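# Animation Blueprint for the CopperToll pickaxe that adds physics motion to its cached input pose.
# bRigidBodyEnabled selects between a RigidBody simulation in b_AR_Root space and a fallback series of five spring-driven AnimDynamics nodes on the dyn_chain bones.
# Each update sets VelScale from the falling, jumping and front-end states, feeding only the RigidBody node's component velocity and acceleration scales.
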AnimBlueprint /BRCosmetics/Weapons/FORT_Melee/Pickaxe_CopperToll/Meshes/CopperToll_Axe_AnimBP
target_skeleton=&/CopperToll_Axe_Skeleton
graphs [
 AnimGraph {nodes [
   n1 = Get VelScale() -> (VelScale:Vector)
   n2 = RigidBody!(
    ComponentPose=Local To Component(LocalPose=Use cached pose 'Input'())
    Alpha
    ExternalForce
    ComponentLinearAccScale:Vector=@n1
    ComponentLinearVelScale:Vector=@n1
   ) -> (Pose) {
    bClampLinearTranslationLimitToRefPose=true
    bOverrideWorldGravity=true
    overridePhysicsAsset=&/CopperToll_Axe_Physics
    simulationSpace=BaseBoneSpace
    alphaBoolBlend {blendOption=Linear}
    baseBoneRef {boneName=b_AR_Root}
    componentAppliedLinearAccClamp=<3000 3000 3000>
    componentLinearAccScale=<0.45 0.45 0.45>
    componentLinearVelScale=<0.45 0.45 0.45>
    overrideWorldGravity {z=-980}
   }
   n3 = Anim Dynamics - Bone: dyn_chain_01(ComponentPose=Local To Component(LocalPose=Use cached pose 'Input'()), Alpha=1, ExternalForce="0,0,-100") {
    angularDampingOverride=0.7
    angularSpringConstant=500
    bAngularSpring=true
    bOverrideAngularDamping=true
    bUseGravityOverride=true
    alphaBoolBlend {blendOption=Linear}
    boundBone {boneName=dyn_chain_01}
    componentAppliedLinearAccClamp=<3000 3000 3000>
    componentLinearAccScale=<0.3 0.3 0.3>
    componentLinearVelScale=<0.3 0.3 0.3>
    gravityOverride {z=-800}
    physicsBodyDefinitions [{
      boundBone {boneName=dyn_chain_01}
      boxExtents=<20 20 10>
      constraintSetup {angularLimitsMax=<5 5 -20>, angularLimitsMin=<-5 -5 -30>, angularTarget=<0 0 0>}
      localJointOffset=<3 0 0>
     }]
   }
   n4 = Anim Dynamics - Bone: dyn_chain_02(ComponentPose=@n3, Alpha=1, ExternalForce="0,0,-100") {
    angularDampingOverride=0.7
    angularSpringConstant=500
    bAngularSpring=true
    bOverrideAngularDamping=true
    bUseGravityOverride=true
    alphaBoolBlend {blendOption=Linear}
    boundBone {boneName=dyn_chain_02}
    componentAppliedLinearAccClamp=<3000 3000 3000>
    componentLinearAccScale=<0.3 0.3 0.3>
    componentLinearVelScale=<0.3 0.3 0.3>
    gravityOverride {z=-800}
    physicsBodyDefinitions [{
      boundBone {boneName=dyn_chain_02}
      boxExtents=<20 20 10>
      constraintSetup {angularLimitsMax=<0 0 30>, angularLimitsMin=<0 -20 -90>, angularTarget=<0 0 0>}
      localJointOffset=<3 0 0>
     }]
    relativeSpaceBone {boneName=b_AR_Root}
   }
   n5 = Anim Dynamics - Bone: dyn_chain_03(ComponentPose=@n4, Alpha=1, ExternalForce="0,0,-100") {
    angularDampingOverride=0.7
    angularSpringConstant=500
    bAngularSpring=true
    bOverrideAngularDamping=true
    bUseGravityOverride=true
    alphaBoolBlend {blendOption=Linear}
    boundBone {boneName=dyn_chain_03}
    componentAppliedLinearAccClamp=<3000 3000 3000>
    componentLinearAccScale=<0.3 0.3 0.3>
    componentLinearVelScale=<0.3 0.3 0.3>
    gravityOverride {z=-800}
    physicsBodyDefinitions [{
      boundBone {boneName=dyn_chain_03}
      boxExtents=<20 20 10>
      constraintSetup {angularLimitsMax=<0 20 20>, angularLimitsMin=<0 -20 -20>, angularTarget=<0 0 0>}
      localJointOffset=<3 0 0>
     }]
    relativeSpaceBone {boneName=b_AR_Root}
   }
   n6 = Anim Dynamics - Bone: dyn_chain_04(ComponentPose=@n5, Alpha=1, ExternalForce="0,0,-100") {
    angularDampingOverride=0.7
    angularSpringConstant=500
    bAngularSpring=true
    bOverrideAngularDamping=true
    bUseGravityOverride=true
    alphaBoolBlend {blendOption=Linear}
    boundBone {boneName=dyn_chain_04}
    componentAppliedLinearAccClamp=<3000 3000 3000>
    componentLinearAccScale=<0.3 0.3 0.3>
    componentLinearVelScale=<0.3 0.3 0.3>
    gravityOverride {z=-800}
    physicsBodyDefinitions [{
      boundBone {boneName=dyn_chain_04}
      boxExtents=<20 20 10>
      constraintSetup {angularLimitsMax=<0 20 20>, angularLimitsMin=<0 -20 -20>, angularTarget=<0 0 0>}
      localJointOffset=<3 0 0>
     }]
    relativeSpaceBone {boneName=b_AR_Root}
   }
   n7 = Anim Dynamics - Bone: dyn_chain_05(ComponentPose=@n6, Alpha=1, ExternalForce="0,0,-100") {
    angularDampingOverride=0.7
    angularSpringConstant=500
    bAngularSpring=true
    bOverrideAngularDamping=true
    bUseGravityOverride=true
    alphaBoolBlend {blendOption=Linear}
    boundBone {boneName=dyn_chain_05}
    componentAppliedLinearAccClamp=<3000 3000 3000>
    componentLinearAccScale=<0.3 0.3 0.3>
    componentLinearVelScale=<0.3 0.3 0.3>
    gravityOverride {z=-800}
    physicsBodyDefinitions [{
      boundBone {boneName=dyn_chain_05}
      boxExtents=<20 20 10>
      constraintSetup {angularLimitsMax=<0 20 20>, angularLimitsMin=<0 -20 -20>, angularTarget=<0 0 0>}
      localJointOffset=<3 0 0>
     }]
    relativeSpaceBone {boneName=b_AR_Root}
   }
   Output Pose(Result=Blend Poses by bool(BlendPose_0=Component To Local(ComponentPose=@n2), BlendPose_1=Component To Local(ComponentPose=@n7), bActiveValue=Get bRigidBodyEnabled()) {blendType=Linear}) {layerGroup=None}
   AnimGraphNode_SubInput_0 = Input Pose()
   Save cached pose 'Input'(Pose=@AnimGraphNode_SubInput_0) {cacheName=Input}
  ]}
 EventGraph {nodes [
   Event BlueprintUpdateAnimation()
   > n1 = Branch(Condition=OR Boolean(A=Get bIsFalling(), B=Get bIsJumping()))
   > Set VelScale(VelScale:Vector="0.4,0.4,0.4") -> (Output_Get:Vector)
   > Branch(Condition=Get bIsFalling())
   > Set VelScale(VelScale="0.35,0.35,0.35")
   Set VelScale(execute=@n1.else, VelScale="0.45,0.45,0.45")
   > Branch(Condition=Get bIsInFrontEnd())
   > Set VelScale(VelScale="0.1,0.25,0.25")
   TryGetPawnOwner()
  ]}
]
variables [VelScale {type=Vector}]
